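# Page-counter widget for the Athena radial HUD, showing pagination text between left and right arrows inside a rounded border.
# ShowUpdatedPagination switches between the bordered, arrowed, grouped look and plain ungrouped text, changing border brush, arrow visibility and font.
# The DropShadow flag also picks the active border brush, shadow offsets and text margin; PreConstruct applies a preview value.
# SelectAnimation plays the Select animation forward or reversed, animating the border material's Size and fading in brush alpha.
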WidgetBlueprint /FortniteCoreUI/UI/InGame/HUD/AthenaRadial/FortTextPaginationEntry
animations [Select {
  binding_count=1
  length_seconds=0.3
  object=&self:Select
  sequence {bindings [RoundedBorder {
     id=#1
     possessed_class=CommonBorder
     tracks [
      {
       type=MovieSceneWidgetMaterialTrack
       sections [{type=MovieSceneParameterSection, end_frame=6, start_frame=0, channels [Size {type=MovieSceneScriptingFloatChannel, keys [[0,0.91],[6,0.8]]}]}]
      }
      {
       type=MovieSceneColorTrack
       display_name=Brush Color
       sections [{
         type=MovieSceneColorSection
         end_frame=6
         start_frame=0
         channels [
          Color.R {type=MovieSceneScriptingFloatChannel, const [1,1]}
          Color.G {type=MovieSceneScriptingFloatChannel, const [1,1]}
          Color.B {type=MovieSceneScriptingFloatChannel, const [1,1]}
          Color.A {type=MovieSceneScriptingFloatChannel, keys [[0,0],[6,1]]}
         ]
        }]
      }
     ]
    }]}
 }]
graphs [
 EventGraph {nodes [
   Event PreConstruct()
   > ShowUpdatedPagination!(execute, self, ShowUpdatedPagination:bool=Get Preview Show Updated Pagination() -> (Preview Show Updated Pagination:bool)) -> (then)
  ]}
 SelectAnimation {nodes [
   n1 = SelectAnimation!() -> (then, Select:bool=~) {is=title K2Node_FunctionEntry}
   > PlayAnimation(
    InAnimation=Get Select() -> (Select:Widget Animation Object Reference)
    PlayMode=Select(Option 0:EUMGSequencePlayMode Enum=Reverse, Option 1:EUMGSequencePlayMode Enum, Index=@n1) -> (ReturnValue:EUMGSequencePlayMode Enum)
   )
  ]}
 ShowUpdatedPagination {nodes [
   n1 = ShowUpdatedPagination!() -> (then, ShowUpdatedPagination:bool=~) {is=title K2Node_FunctionEntry}
   n2 = Select(Option 0:ESlateVisibility Enum=Collapsed, Option 1:ESlateVisibility Enum=HitTestInvisible, Index=@n1) -> (ReturnValue:ESlateVisibility Enum)
   > SetVisibility(self=Get Arrow_Right() -> (Arrow_Right:Common Text Object Reference), InVisibility=@n2)
   > SetVisibility(self=Get Arrow_Left() -> (Arrow_Left:Common Text Object Reference), InVisibility=@n2)
   n3 = Get DropShadow() -> (DropShadow:bool)
   n4 = Select(Option 0:Vector 2D Structure, Option 1:Vector 2D Structure=Get ArrowShadowOffset() -> (ArrowShadowOffset:Vector 2D Structure), Index=@n3) -> (ReturnValue:Vector 2D Structure)
   > SetShadowOffset!(execute, self:Text Object Reference=Get Arrow_Right(), InShadowOffset:Vector 2D Structure=@n4) -> (then)
   > SetShadowOffset(self=Get Arrow_Left(), InShadowOffset=@n4)
   n5 = Get ActiveDropShadowBorderBrush() -> (ActiveDropShadowBorderBrush:Slate Brush Structure)
   n6 = Get ActiveBorderBrush() -> (ActiveBorderBrush:Slate Brush Structure)
   > SetBrush(
    self:Border Object Reference=Get RoundedBorder() -> (RoundedBorder:Common Border Object Reference)
    InBrush=Select(
     Option 0:"Slate Brush Structure (by ref)"=Get InactiveBorderBrush() -> (InactiveBorderBrush:Slate Brush Structure)
     Option 1:"Slate Brush Structure (by ref)"=Select(Option 0:Slate Brush Structure=@n6, Option 1:Slate Brush Structure=@n5, Index=@n3) -> (ReturnValue:Slate Brush Structure)
     Index=@n1
    ) -> (ReturnValue:"Slate Brush Structure (by ref)")
   )
   n7 = Get Text_Pagination() -> (Text_Pagination:Common Text Object Reference)
   > Set Font!(
    execute
    Font:Slate Font Info Structure=Select(
     Option 0:Slate Font Info Structure=Get UnGroupedFontInfo() -> (UnGroupedFontInfo:Slate Font Info Structure)
     Option 1:Slate Font Info Structure=Get GroupedFontInfo() -> (GroupedFontInfo:Slate Font Info Structure)
     Index=@n1
    ) -> (ReturnValue:Slate Font Info Structure)
    self:Text Object Reference=@n7
   ) -> (then, Output_Get:Slate Font Info Structure)
   n8 = Get UnGroupedShadowOffset() -> (UnGroupedShadowOffset:Vector 2D Structure)
   n9 = Get GroupedShadowOffset() -> (GroupedShadowOffset:Vector 2D Structure)
   n10 = Get DropShadow()
   > SetShadowOffset(
    self=@n7
    InShadowOffset=Select(Option 0:Vector 2D Structure, Option 1:Vector 2D Structure=Select(Option 0=@n8, Option 1=@n9, Index=@n1), Index=@n10) -> (ReturnValue:Vector 2D Structure)
   )
   > SetMargin!(
    execute
    self:Common Text Object Reference=Get Text_Pagination()
    InMargin:"Margin Structure (by ref)"=Select(
     Option 0:Margin Structure=Get PaddingTextPagination() -> (PaddingTextPagination:Margin Structure)
     Option 1:Margin Structure=Get PaddingTextPaginationDropShadow() -> (PaddingTextPaginationDropShadow:Margin Structure)
     Index=@n10
    ) -> (ReturnValue:Margin Structure)
   ) -> (then)
  ]}
]
variables [
 ActiveBorderBrush {
  type=SlateBrush
  default {
   bIsDynamicallyLoaded=false
   drawAs=Image
   imageType=FullColor
   mirroring=NoMirror
   resourceName=None
   resourceObject=/Game/Athena/HUD/LTM/Mash/Materials/MI_UI_Border_PageMarker
   tiling=NoTile
   imageSize=<0 0>
   margin=<ltrb 0 0 0 0>
   outlineSettings {
    bUseBrushTransparency=false
    roundingType=HalfHeightRadius
    width=0
    color {colorUseRule=UseColor_Specified, specifiedColor=<rgba 0 0 0 0>}
    cornerRadii=<0 0 0 0>
   }
   tintColor {colorUseRule=UseColor_Specified, specifiedColor=<rgba 1 1 1 1>}
   uVRegion {bIsValid=false, max=<0 0>, min=<0 0>}
  }
 }
 ActiveDropShadowBorderBrush {
  type=SlateBrush
  default {
   bIsDynamicallyLoaded=false
   drawAs=Image
   imageType=FullColor
   mirroring=NoMirror
   resourceName=None
   resourceObject=/Game/Athena/HUD/LTM/Mash/Materials/MI_UI_Border_PageMarker_DropShadow
   tiling=NoTile
   imageSize=<0 0>
   margin=<ltrb 0 0 0 0>
   outlineSettings {
    bUseBrushTransparency=false
    roundingType=HalfHeightRadius
    width=0
    color {colorUseRule=UseColor_Specified, specifiedColor=<rgba 0 0 0 0>}
    cornerRadii=<0 0 0 0>
   }
   tintColor {colorUseRule=UseColor_Specified, specifiedColor=<rgba 1 1 1 1>}
   uVRegion {bIsValid=false, max=<0 0>, min=<0 0>}
  }
 }
 ArrowShadowOffset {type=Vector2D, default=<1.2 1.5>}
 DropShadow {type=boolean, default=true}
 GroupedFontInfo {
  type=SlateFontInfo
  default {
   bForceMonospaced=false
   bMaterialIsStencil=false
   fontMaterial=null
   fontObject=&Burbank
   letterSpacing=0
   monospacedWidth=1
   size=18.75
   skewAmount=0
   typefaceFontName=Burbank-Black
   outlineSettings {
    bApplyOutlineToDropShadows=false
    bMiteredCorners=false
    bSeparateFillAlpha=false
    outlineMaterial=null
    outlineSize=3
    outlineColor=<rgba 0 0 0 0>
   }
  }
 }
 GroupedShadowOffset {type=Vector2D, default=<4 2.5>}
 InactiveBorderBrush {
  type=SlateBrush
  default {
   bIsDynamicallyLoaded=false
   drawAs=NoDrawType
   imageType=NoImage
   mirroring=NoMirror
   resourceName=None
   resourceObject=null
   tiling=NoTile
   imageSize=<0 0>
   margin=<ltrb 0 0 0 0>
   outlineSettings {
    bUseBrushTransparency=false
    roundingType=HalfHeightRadius
    width=0
    color {colorUseRule=UseColor_Specified, specifiedColor=<rgba 0 0 0 0>}
    cornerRadii=<0 0 0 0>
   }
   tintColor {colorUseRule=UseColor_Specified, specifiedColor=<rgba 1 1 1 0>}
   uVRegion {bIsValid=false, max=<0 0>, min=<0 0>}
  }
 }
 PaddingTextPagination {type=Margin, default=<ltrb 2 2 1.25 2>}
 PaddingTextPaginationDropShadow {type=Margin, default=<ltrb 0 2 -1.75 0>}
 Preview Show Updated Pagination {type=boolean}
 ShowBorder {type=boolean, default=true}
 UnGroupedFontInfo {
  type=SlateFontInfo
  default {
   bForceMonospaced=false
   bMaterialIsStencil=false
   fontMaterial=null
   fontObject=&Burbank
   letterSpacing=0
   monospacedWidth=1
   size=30
   skewAmount=0
   typefaceFontName=Burbank-Black
   outlineSettings {
    bApplyOutlineToDropShadows=false
    bMiteredCorners=false
    bSeparateFillAlpha=false
    outlineMaterial=&M-FontShadow
    outlineSize=3
    outlineColor=<rgba 0 0 0 0.5>
   }
  }
 }
 UnGroupedShadowOffset {type=Vector2D, default=<5 5>}
]
widgets {
 info {inheritedWidgetCount=0, parentClass=&UserWidget, rootWidgetClass=&Overlay, widgetCount=6}
 widgets [
  Overlay_0 {widgetClassPath=&Overlay, widget=Overlay_0}
  RoundedBorder {
   widgetClassPath=&CommonBorder
   bIsVariable=true
   parent=Overlay_0
   widget=RoundedBorder
   properties {
    horizontalAlignment=HAlign_Center
    verticalAlignment=VAlign_Center
    background {
     bIsDynamicallyLoaded=false
     drawAs=NoDrawType
     imageType=NoImage
     mirroring=NoMirror
     resourceName=None
     resourceObject=null
     tiling=NoTile
     imageSize=<0 0>
     margin=<ltrb 0 0 0 0>
     outlineSettings {
      bUseBrushTransparency=false
      roundingType=HalfHeightRadius
      width=0
      color {colorUseRule=UseColor_Specified, specifiedColor=<rgba 0 0 0 0>}
      cornerRadii=<0 0 0 0>
     }
     tintColor {colorUseRule=UseColor_Specified, specifiedColor=<rgba 1 1 1 0>}
     uVRegion {bIsValid=false, max=<0 0>, min=<0 0>}
    }
    padding=<ltrb 9.5 0 9 0>
   }
   slot {class=OverlaySlot}
  }
  HorizontalBox_0 {
   widgetClassPath=&HorizontalBox
   parent=RoundedBorder
   widget=HorizontalBox_0
   slot {class=BorderSlot, properties {horizontalAlignment=HAlign_Center, verticalAlignment=VAlign_Center, padding=<ltrb 9.5 0 9 0>}}
  }
  Arrow_Left {
   widgetClassPath=&CommonTextBlock
   bIsVariable=true
   parent=HorizontalBox_0
   widget=Arrow_Left
   properties {
    justification=Center
    text="<"
    textTransformPolicy=ToUpper
    font {
     bForceMonospaced=false
     bMaterialIsStencil=false
     fontMaterial=null
     fontObject=&Burbank
     letterSpacing=0
     monospacedWidth=1
     size=24
     skewAmount=0
     typefaceFontName=Burbank-Black
     outlineSettings {
      bApplyOutlineToDropShadows=false
      bMiteredCorners=false
      bSeparateFillAlpha=false
      outlineMaterial=null
      outlineSize=0
      outlineColor=<rgba 0 0 0 1>
     }
    }
    shadowColorAndOpacity=<rgba 0 0 0 0.502>
    shadowOffset=<1.2 1.5>
   }
   slot {class=HorizontalBoxSlot, properties {horizontalAlignment=HAlign_Center, verticalAlignment=VAlign_Center, padding=<ltrb 0 -5 0 -5>}}
  }
  Text_Pagination {
   widgetClassPath=&CommonTextBlock
   bIsVariable=true
   parent=HorizontalBox_0
   widget=Text_Pagination
   properties {
    justification=Center
    text=2/5
    textTransformPolicy=ToUpper
    font {
     bForceMonospaced=false
     bMaterialIsStencil=false
     fontMaterial=null
     fontObject=&Burbank
     letterSpacing=0
     monospacedWidth=1
     size=18.75
     skewAmount=0
     typefaceFontName=Burbank-Black
     outlineSettings {
      bApplyOutlineToDropShadows=false
      bMiteredCorners=false
      bSeparateFillAlpha=false
      outlineMaterial=null
      outlineSize=3
      outlineColor=<rgba 0 0 0 0>
     }
    }
    renderTransform {angle=0, scale=<1 1>, shear=<-8 0>, translation=<0 0>}
    shadowColorAndOpacity=<rgba 0 0 0 0.502>
    shadowOffset=<4 2.5>
   }
   slot {class=HorizontalBoxSlot}
  }
  Arrow_Right {
   widgetClassPath=&CommonTextBlock
   bIsVariable=true
   parent=HorizontalBox_0
   widget=Arrow_Right
   properties {
    justification=Center
    text=">"
    textTransformPolicy=ToUpper
    font {
     bForceMonospaced=false
     bMaterialIsStencil=false
     fontMaterial=null
     fontObject=&Burbank
     letterSpacing=0
     monospacedWidth=1
     size=24
     skewAmount=0
     typefaceFontName=Burbank-Black
     outlineSettings {
      bApplyOutlineToDropShadows=false
      bMiteredCorners=false
      bSeparateFillAlpha=false
      outlineMaterial=null
      outlineSize=0
      outlineColor=<rgba 0 0 0 1>
     }
    }
    shadowColorAndOpacity=<rgba 0 0 0 0.502>
    shadowOffset=<1.2 1.5>
   }
   slot {class=HorizontalBoxSlot, properties {horizontalAlignment=HAlign_Center, verticalAlignment=VAlign_Center, padding=<ltrb 0 -5 0 -5>}}
  }
 ]
}
