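# Radial picker HUD pagination widget pairing a dotted entry with a "page / total" text entry, with category grouping controlled by the GroupEmotePickerCategories rollback CVar.
# PaginationUpdated stores the page and refreshes counter text and arrow dimming.
# When grouped, page zero selects the dotted entry and later pages select the text entry.
# Text selection plays one animation forward or reversed, with border updates triggered by both finished events and backup timers.
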WidgetBlueprint /FortniteCoreUI/UI/InGame/HUD/AthenaRadial/FortTextPagination
graphs [
 EventGraph {nodes [
   ResetTextPaginationState() {is=title K2Node_CustomEvent}
   > Set Text_Selected(Text_Selected:bool=false) -> (Output_Get:bool)
   > UpdateTextPaginationBorderOpacity!(execute, self, ShowBorder:bool=false) -> (then)
   OnTextPaginationSelected() {is=title K2Node_CustomEvent}
   > Set Text_Selected(Text_Selected=true)
   > SetBrushColor(
    self=Get RoundedBorder!(self:Fort Text Pagination Entry Object Reference=Get FortTextPaginationEntry() -> (FortTextPaginationEntry:Fort Text Pagination Entry Object Reference)) -> (RoundedBorder:Common Border Object Reference)
    InBrushColor=(R=1,G=1,B=1,A=1)
   )
   > n1 = Branch(Condition=IsAnimationPlaying(self=Get FortTextPaginationEntry(), InAnimation=Get TextSelectAnim() -> (TextSelectAnim:Widget Animation Object Reference)))
   > n2 = StopAnimation(self=Get FortTextPaginationEntry(), InAnimation=Get TextSelectAnim())
   n3 = "Play Animation with Finished event (legacy)"!(
    execute:Exec=[@n1.else, @n2.then]
    Widget:User Widget Object Reference=Get FortTextPaginationEntry()
    InAnimation:Widget Animation Object Reference=Get TextSelectAnim()
    StartAtTime:float=0
    NumLoopsToPlay:int=1
    PlayMode:EUMGSequencePlayMode Enum
    PlaybackSpeed:float=1
   ) -> (then:Exec, Result:UMGSequence Player Object Reference, Finished:Exec) {is=title K2Node_PlayAnimation}
   n4 = OnSelectAnimationComplete() {is=title K2Node_CustomEvent}
   Set Timer by Event(execute=@n3.then, Delegate=@n4, Time=0.3, bLooping=false, bMaxOncePerFrame=true)
   OnSelectAnimationComplete(execute=@n3.Finished) {is=title K2Node_CallFunction}
   n5 = OnDeselectAnimationComplete() {is=title K2Node_CustomEvent}
   > UpdateTextPaginationBorderOpacity(ShowBorder=false)
   OnTextPaginationDeselected() {is=title K2Node_CustomEvent}
   > Set Text_Selected(Text_Selected=false)
   > n6 = Branch(Condition=IsAnimationPlaying(self=Get FortTextPaginationEntry(), InAnimation=Get TextSelectAnim()))
   > n7 = StopAnimation(self=Get FortTextPaginationEntry(), InAnimation=Get TextSelectAnim())
   n8 = "Play Animation with Finished event (legacy)"(
    execute=[@n6.else, @n7.then]
    Widget=Get FortTextPaginationEntry()
    InAnimation=Get TextSelectAnim()
    StartAtTime=0
    NumLoopsToPlay=1
    PlayMode=Reverse
    PlaybackSpeed=1
   )
   > Set Timer by Event(Delegate=@n5, Time=0.3, bLooping=false, bMaxOncePerFrame=true)
   OnDeselectAnimationComplete(execute=@n8.Finished) {is=title K2Node_CallFunction}
   Event PreConstruct()
   > Set TextSelectAnim(TextSelectAnim:Widget Animation Object Reference=Get Select!(self:Fort Text Pagination Entry Object Reference=Get FortTextPaginationEntry()) -> (Select:Widget Animation Object Reference)) -> (Output_Get:Widget Animation Object Reference)
   > Set DottedSelectAnim(DottedSelectAnim:Widget Animation Object Reference=Get Select(self:Fort Dotted Pagination Entry Object Reference=Get FortDottedPaginationEntry() -> (FortDottedPaginationEntry:Fort Dotted Pagination Entry Object Reference))) -> (Output_Get:Widget Animation Object Reference)
   > Set DottedDeselectAnim(DottedDeselectAnim:Widget Animation Object Reference=Get Deselect!(self:Fort Dotted Pagination Entry Object Reference=Get FortDottedPaginationEntry()) -> (Deselect:Widget Animation Object Reference)) -> (Output_Get:Widget Animation Object Reference)
   > SetMobileStyle() {is=title K2Node_CallFunction}
   K2Node_Event_2 = PaginationUpdated!() -> (OutputDelegate, then, PageIndex:int, ShortcutPageIndex:int, NumPagesInCategory:int, CategoryChanged:bool=~) {is=title K2Node_CustomEvent}
   > Set Page Index(Page Index:int=@K2Node_Event_2.PageIndex) -> (Output_Get:int=0)
   > UpdateTextPagination!(
    execute
    self
    PageIndex:int=Get Page Index() -> (Page Index:int=0)
    ShortcutPageIndex:int=@K2Node_Event_2.ShortcutPageIndex
    NumPagesInCategory:int=@K2Node_Event_2.NumPagesInCategory
   ) -> (then)
   > Branch(Condition=IsGrouppingCategories!() -> (ReturnValue:bool))
   > n9 = Branch(Condition=@K2Node_Event_2.CategoryChanged)
   > ResetSelectedState(self:Fort Dotted Pagination Entry Object Reference=Get FortDottedPaginationEntry())
   > n10 = ResetTextPaginationState() {is=title K2Node_CallFunction}
   n11 = "Equal (Integer)"(A=Get Page Index() -> (Page Index=0))
   SetIsSelected!(
    execute=[@n10.then, @n9.else]
    self:Common Button Base Object Reference=Get FortDottedPaginationEntry()
    InSelected:bool=@n11
    bGiveClickFeedback:bool=true
   ) -> (then)
   > n12 = Branch(Condition=NOT Boolean(A=@n11))
   n13 = Get Text_Selected() -> (Text_Selected:bool)
   n14 = Get Page Index() -> (Page Index=0)
   > Branch(Condition=AND Boolean(A="integer > integer"(A=@n14), B=NOT Boolean(A=@n13)))
   > OnTextPaginationSelected() {is=title K2Node_CallFunction}
   Branch(execute=@n12.else, Condition=Get Text_Selected())
   > OnTextPaginationDeselected() {is=title K2Node_CallFunction}
   Event Construct()
   n15 = IsGrouppingCategories()
   > ShowUpdatedPagination!(execute, self:Fort Text Pagination Entry Object Reference=Get FortTextPaginationEntry(), ShowUpdatedPagination:bool=@n15) -> (then)
   > SetVisibility(self=Get FortDottedPaginationEntry(), InVisibility=Select(Option 0=Collapsed, Option 1=HitTestInvisible, Index=@n15))
   n16 = Get Page Index() -> (Page Index=0)
   > UpdateTextPaginationBorderOpacity(ShowBorder=AND Boolean(A="integer > integer"(A=@n16), B=IsGrouppingCategories()))
   > n17 = Branch(Condition="Equal (Integer)"(A=Get Page Index() -> (Page Index=0)))
   > n18 = Branch(Condition=IsAnimationPlaying(self=Get FortDottedPaginationEntry(), InAnimation=Get DottedSelectAnim() -> (DottedSelectAnim:Widget Animation Object Reference)))
   > n19 = StopAnimation(self=Get FortDottedPaginationEntry(), InAnimation=Get DottedSelectAnim())
   PlayAnimation(execute=[@n21.then, @n20.else, @n19.then], self=Get FortDottedPaginationEntry(), InAnimation=Get DottedSelectAnim())
   n20 = Branch(
    execute=@n18.else
    Condition=IsAnimationPlaying(self=Get FortDottedPaginationEntry(), InAnimation=Get DottedDeselectAnim() -> (DottedDeselectAnim:Widget Animation Object Reference))
   )
   > n21 = StopAnimation(self=Get FortDottedPaginationEntry(), InAnimation=Get DottedDeselectAnim())
   Branch(execute=@n17.else, Condition="integer > integer"(A=Get Page Index() -> (Page Index=0)))
   > OnTextPaginationSelected()
   UpdateTextPaginationBorderOpacity(execute=@n4.then, ShowBorder=true)
  ]}
 GetCurrentPageIndex {nodes [
   Inputs!()
   n1 = NOT Boolean(A=Get UpdatePageIndexWhileOnShortcut() -> (UpdatePageIndexWhileOnShortcut:bool))
   n2 = "To Boolean (Integer)"!(InInt:int=Get PageIndex() -> (PageIndex:int=0)) -> (ReturnValue:bool)
   Outputs!(ReturnValue:int=SelectInt(A=Get PageIndex() -> (PageIndex=0), B=Get ShortcutPageIndex() -> (ShortcutPageIndex:int=0), bPickA=OR Boolean(A=@n2, B=@n1)))
  ]}
 IsGrouppingCategories {nodes [
   IsGrouppingCategories() {is=title K2Node_FunctionEntry}
   > Return Node!(
    execute
    ReturnValue:bool="Equal (Float)"(A=GetNumericCVar!(CVar:String=Fort.Rollback.GroupEmotePickerCategories) -> (OutValue:float, ReturnValue:bool), B=1)
   )
  ]}
 SetMobileStyle {nodes [
   SetMobileStyle() {is=title K2Node_FunctionEntry}
   n1 = GetOwningPlayer()
   > Branch(Condition=AND Boolean(A=IsMobileGame(), B=IsUsingTouch(PlayerController=@n1)))
   > SetPadding!(
    execute
    self:Horizontal Box Slot Object Reference=SlotAsHorizontalBoxSlot!(Widget:Widget Object Reference=Get FortDottedPaginationEntry()) -> (ReturnValue:Horizontal Box Slot Object Reference)
    InPadding_Left:float=0
    InPadding_Top:float=0
    InPadding_Right:float=5
    InPadding_Bottom:float=0
   ) -> (then)
  ]}
 UpdateTextPagination {nodes [
   UpdateTextPagination!() -> (then, PageIndex:int, ShortcutPageIndex:int, NumPagesInCategory:int) {is=title K2Node_FunctionEntry}
   n1 = int - int(
    A=Get NumPagesInCategory() -> (NumPagesInCategory:int=0)
    B=SelectInt(A=1, B=0, bPickA=IsGrouppingCategories() {is=title K2Node_CallFunction})
   )
   n2 = "To String (Integer)"(InInt=@n1)
   n3 = SelectInt(A=0, B=1, bPickA=IsGrouppingCategories())
   n4 = Get Current Page Index!() -> (ReturnValue:int) {is=title K2Node_MacroInstance}
   n5 = "To String (Integer)"(InInt="Max (Integer)"(A=int + int(A=@n3, B=@n4), B=1))
   > Set Text(
    Text="To Text (String)"(InString=Append!(A=@n5, B=" / ", C:String=@n2) -> (ReturnValue))
    self=Get Text_Pagination!(self:Fort Text Pagination Entry Object Reference=Get FortTextPaginationEntry()) -> (Text_Pagination:Common Text Object Reference)
   )
   n6 = IsGrouppingCategories()
   n7 = Get Current Page Index()
   n8 = OR Boolean(A="Equal (Integer)"(A=@n7, B=1), B="Equal (Integer)"(A=@n7))
   > SetColorAndOpacity(
    self=Get Arrow_Left!(self:Fort Text Pagination Entry Object Reference=Get FortTextPaginationEntry()) -> (Arrow_Left:Common Text Object Reference)
    InColorAndOpacity=Select(
     Option 0:Slate Color Structure=Get Acive_Color() -> (Acive_Color:Slate Color Structure)
     Option 1:Slate Color Structure=Get Inactive_Color() -> (Inactive_Color:Slate Color Structure)
     Index=AND Boolean(A=@n8, B=@n6)
    ) -> (ReturnValue:Slate Color Structure)
   )
   n9 = IsGrouppingCategories()
   n10 = "Equal (Integer)"(A=Get Current Page Index(), B=@n1)
   > SetColorAndOpacity(
    self=Get Arrow_Right!(self:Fort Text Pagination Entry Object Reference=Get FortTextPaginationEntry()) -> (Arrow_Right:Common Text Object Reference)
    InColorAndOpacity=Select(Option 0=Get Acive_Color(), Option 1=Get Inactive_Color(), Index=AND Boolean(A=@n10, B=@n9))
   )
  ]}
 UpdateTextPaginationBorderOpacity {nodes [
   n1 = UpdateTextPaginationBorderOpacity!() -> (then, ShowBorder:bool=~) {is=title K2Node_FunctionEntry}
   > Set BrushColor!(
    execute
    BrushColor:Linear Color Structure=Select(Option 0:Linear Color Structure=(R=0,G=0,B=0,A=0), Option 1:Linear Color Structure=(R=1,G=1,B=1,A=1), Index=@n1) -> (ReturnValue:Linear Color Structure)
    self:Border Object Reference=Get RoundedBorder(self=Get FortTextPaginationEntry())
   ) -> (then, Output_Get:Linear Color Structure)
  ]}
]
variables [
 Acive_Color {type=SlateColor, default {colorUseRule=UseColor_Specified, specifiedColor=<rgba 1 1 1 1>}}
 DottedDeselectAnim {type=WidgetAnimation, default=null}
 DottedSelectAnim {type=WidgetAnimation, default=null}
 DropShadow {type=boolean, default=true}
 Inactive_Color {type=SlateColor, default {colorUseRule=UseColor_Specified, specifiedColor=<rgba 1 1 1 0.5>}}
 Page Index {type=integer}
 Text_Selected {type=boolean}
 TextSelectAnim {type=WidgetAnimation, default=null}
 UpdatePageIndexWhileOnShortcut {type=boolean, default=true}
]
widgets {
 info {inheritedWidgetCount=0, parentClass=&CommonUserWidget, rootWidgetClass=&ScaleBox, widgetCount=4}
 widgets [
  ScaleBox_0 {widgetClassPath=&ScaleBox, bIsVariable=true, widget=ScaleBox_0, properties {stretch=UserSpecified}}
  HorizontalBox {widgetClassPath=&HorizontalBox, bIsVariable=true, parent=ScaleBox_0, widget=HorizontalBox, slot {class=ScaleBoxSlot}}
  FortDottedPaginationEntry {
   widgetClassPath=/Game/UI/InGame/HUD/AthenaRadial/FortDottedPaginationEntry
   bIsVariable=true
   parent=HorizontalBox
   widget=FortDottedPaginationEntry
   properties {
    bSelectable=true
    bToggleable=true
    holdData=&CommonUI_HoldData_Base_C
    onButtonBaseClicked=(&self:WidgetTree.FortDottedPaginationEntry.HandleOnButtonClicked)
    onButtonBaseDoubleClicked=(&self:WidgetTree.FortDottedPaginationEntry.HandleOnButtonDoubleClicked)
    onButtonBaseHovered=(&self:WidgetTree.FortDottedPaginationEntry.HandleOnButtonHovered)
    onButtonBaseUnhovered=(&self:WidgetTree.FortDottedPaginationEntry.HandleOnButtonUnhovered)
    onSelectedChangedBase=(&self:WidgetTree.FortDottedPaginationEntry.HandleOnSelectedChanged)
   }
   slot {class=HorizontalBoxSlot}
  }
  FortTextPaginationEntry {
   widgetClassPath=&/FortTextPaginationEntry
   bIsVariable=true
   parent=HorizontalBox
   widget=FortTextPaginationEntry
   properties {
    showBorder=false
    groupedFontInfo {
     bForceMonospaced=false
     bMaterialIsStencil=false
     fontMaterial=null
     fontObject=&Burbank
     letterSpacing=0
     monospacedWidth=1
     size=19.5
     skewAmount=0
     typefaceFontName=Burbank-Black
     outlineSettings {
      bApplyOutlineToDropShadows=false
      bMiteredCorners=false
      bSeparateFillAlpha=false
      outlineMaterial=null
      outlineSize=3
      outlineColor=<rgba 0 0 0 0>
     }
    }
    inactiveBorderBrush {
     bIsDynamicallyLoaded=false
     drawAs=Image
     imageType=NoImage
     mirroring=NoMirror
     resourceName=None
     resourceObject=null
     tiling=NoTile
     imageSize=<0 0>
     margin=<ltrb 0 0 0 0>
     outlineSettings {
      bUseBrushTransparency=false
      roundingType=HalfHeightRadius
      width=0
      color {colorUseRule=UseColor_Specified, specifiedColor=<rgba 0 0 0 0>}
      cornerRadii=<0 0 0 0>
     }
     tintColor {colorUseRule=UseColor_Specified, specifiedColor=<rgba 1 1 1 0>}
     uVRegion {bIsValid=false, max=<0 0>, min=<0 0>}
    }
    paddingTextPagination=<ltrb 2 2 2 2>
    unGroupedFontInfo {
     bForceMonospaced=false
     bMaterialIsStencil=false
     fontMaterial=null
     fontObject=&Burbank
     letterSpacing=0
     monospacedWidth=1
     size=30
     skewAmount=0
     typefaceFontName=Burbank-Black
     outlineSettings {
      bApplyOutlineToDropShadows=false
      bMiteredCorners=false
      bSeparateFillAlpha=false
      outlineMaterial=null
      outlineSize=0
      outlineColor=<rgba 0 0 0 1>
     }
    }
   }
   slot {class=HorizontalBoxSlot, properties {horizontalAlignment=HAlign_Center, verticalAlignment=VAlign_Center}}
  }
 ]
}
