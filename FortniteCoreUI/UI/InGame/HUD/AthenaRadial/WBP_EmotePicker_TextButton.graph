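# Text button widget for the Athena radial emote picker HUD, derived from FortTextButton, with placeholder text suggesting category titles.
# It re-lays out the inherited Text_Content inside a scaled-down, fixed-width SizeBox that clips to bounds, so long titles truncate with an ellipsis.
# The label uses outlined, shadowed Burbank text with scrolling disabled.
# The event graph adds no logic beyond parent calls for PreConstruct, Construct and Tick.
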WidgetBlueprint /FortniteCoreUI/UI/InGame/HUD/AthenaRadial/WBP_EmotePicker_TextButton
graphs [EventGraph {nodes [
   n1 = Event PreConstruct()
   > Parent: PreConstruct(IsDesignTime=@n1.IsDesignTime)
   Event Construct()
   > Parent: Construct()
   n2 = Event Tick()
   > Parent: Tick!(execute, MyGeometry:Geometry Structure=@n2.MyGeometry, InDeltaTime:float=@n2.InDeltaTime) -> (then) {is=title K2Node_CallParentFunction}
  ]}]
variables []
widgets {
 info {
  inheritedWidgetCount=2
  parentClass=/FortUILibrary/Foundation/Buttons/FortTextButton.FortTextButton_C
  rootWidgetClass=&ScaleBox
  widgetCount=4
 }
 widgets [
  ScaleBox_0 {widgetClassPath=&ScaleBox, widget=ScaleBox_0, properties {stretch=UserSpecified, userSpecifiedScale=0.9}}
  SizeBox_0 {
   widgetClassPath=&SizeBox
   parent=ScaleBox_0
   widget=SizeBox_0
   properties {bOverride_WidthOverride=true, clipping=ClipToBounds, widthOverride=290}
   slot {class=ScaleBoxSlot}
  }
  Text_Content {
   widgetClassPath=&CommonTextBlock
   bInherited=true
   parent=SizeBox_0
   widget=Text_Content
   properties {
    bIsScrollingEnabled=false
    text="{Long Title Category}"
    textOverflowPolicy=Ellipsis
    font {
     bForceMonospaced=false
     bMaterialIsStencil=false
     fontMaterial=null
     fontObject=&Burbank
     letterSpacing=0
     monospacedWidth=1
     size=16
     skewAmount=0
     typefaceFontName=BurbankRegular-Black
     outlineSettings {
      bApplyOutlineToDropShadows=false
      bMiteredCorners=false
      bSeparateFillAlpha=false
      outlineMaterial=&M-FontShadow
      outlineSize=3
      outlineColor=<rgba 0 0 0 0.5>
     }
    }
    shadowColorAndOpacity=<rgba 0 0 0 0.5>
    shadowOffset=<3 3>
   }
   slot {class=SizeBoxSlot, properties {horizontalAlignment=HAlign_Center, verticalAlignment=VAlign_Center, padding=<ltrb 4 2 4 2>}}
  }
  InputActionWidget {widgetClassPath=&CommonActionWidget, bInherited=true, widget=null}
 ]
}
